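# Niagara system for the "Mangaoutline" drift trail vehicle cosmetic, drawing manga-style bursts, smoke, ribbons and wheel wraps at the wheels.
# Emitters attach by partition or ribbon index to vehicle skeletal mesh locations, and FN_DisableTireFX masks them per tire.
# Spawn groups layer variants within emitters, and painted-variant overrides set colors.
# The two near-duplicate smoke emitters use SubUV sprites and switch rates, velocity and curves for frontend preview.
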NiagaraSystem /VehicleCosmetics/DriftTrail/DriftTrail_Mangaoutline/FX/NS_DriftTrail_Mangaoutline
dependencies {
 usedDataInterfaces [
  &NiagaraDataInterfaceCurve
  &NiagaraDataInterfaceSpriteRendererInfo
  &NiagaraDataInterfaceSkeletalMesh
 ]
 usedDynamicInputs []
 usedModules [
  &self:SystemSpawnScript.SystemScriptSource.SystemScriptGraph.NiagaraNodeAssignment_0.SetVariables_#1
  &FN_UnpackVehicleFlags
  /VehicleCosmetics/FX/Niagara/Modules/VC_SetVehicleSpeedAttribs
  &SystemState
 ]
}
emitters {
 NE_Manga_Burst001 {
  emitterUpdateScript [
   &EmitterState(Inactive Response=::NewEnumerator1, Life Cycle Mode=::NewEnumerator1, Loop Duration=0.1, UseLoopDelay=false)
   &SpawnRate(SpawnRate=16, Use Spawn Probability=false) {enabled=false}
   &SpawnRate(SpawnRate=20, Use Spawn Probability=false) {module=SpawnRate002}
   &SpawnRate(SpawnGroup=1, SpawnRate=20, Use Spawn Probability=false) {module=SpawnRate001}
  ]
  particleSpawnScript [
   &VC_PaintedVariantsOverride(
    Black=<rgba 0.036 0.036 0.04 1>
    BurntSienna=<rgba 0.19 0.0539 0.019 1>
    Cobalt=<rgba 0.072 0.1336 0.6 1>
    Crimson=<rgba 0.43 0.0086 0.0086 1>
    DefaultColor=/VehicleCosmetics/FX/Niagara/Modules/MergeEmitters/VC_Dynamic_ColorBySpawnGroup(
     Color SG 0:LinearColor=<rgba 1 1 1 1>
     MaximumSpawnGroups:NiagaraInt32=2
     Color SG 1:LinearColor=<rgba 1 1 1 1>
     Color SG 2:LinearColor=<rgba 1 1 1 1>
     Color SG 3:LinearColor=<rgba 1 1 1 1>
     Color SG 4:LinearColor=<rgba 1 1 1 1>
    )
    ForestGreen=<rgba 0.0784 0.6 0.0204 1>
    Grey=<rgba 0.1995 0.1995 0.21 1>
    Lime=<rgba 0.9167 1 0 1>
    Orange=<rgba 1 0.5 0 1>
    Pink=<rgba 1 0.32 0.83 1>
    Purple=<rgba 0.2492 0 0.65 1>
    Saffron=<rgba 0.675 0.9 0 1>
    SkyBlue=<rgba 0 0.5323 0.8 1>
    TitaniumWhite=<rgba 1 1 1 1>
   )
   &InitializeParticle(
    Color Mode=::NewEnumerator0
    Lifetime=/VehicleCosmetics/FX/Niagara/Modules/MergeEmitters/VC_FloatBySpawnGroup(
     SpawnGroup or ExecIndex:NiagaraInt32=0
     CustomInt:NiagaraInt32=0
     MaximumSpawnGroups:NiagaraInt32=2
     Float SG 0:NiagaraFloat=0.1
     Float SG 1:NiagaraFloat=0.1
     Float SG 2:NiagaraFloat=0.1
     Float SG 3:NiagaraFloat=0
     Float SG 4:NiagaraFloat=0
     Float SG 5:NiagaraFloat=0
    )
    Mesh Scale Max=&Multiply_VectorByFloat(
     Vector=/VehicleCosmetics/FX/Niagara/Modules/MergeEmitters/VC_VectorBySpawnGroup(
      SpawnGroup or ExecIndex:NiagaraInt32=0
      CustomInt:NiagaraInt32=0
      MaximumSpawnGroups:NiagaraInt32=2
      Vector Group 0:Vector3f=<0.7 0.9 0.5>
      Vector Group 1:Vector3f=<0.7 0.65 0.5>
      Vector Group 2:Vector3f=<0 0 0>
      Vector Group 3:Vector3f=<0 0 0>
      Vector Group 4:Vector3f=<0 0 0>
      Vector Group 5:Vector3f=<0 0 0>
     )
     Float=3.5
    )
    Mesh Scale Min=&Multiply_VectorByFloat(
     Vector=&VC_VectorBySpawnGroup(
      SpawnGroup or ExecIndex:NiagaraInt32=0
      CustomInt:NiagaraInt32=0
      MaximumSpawnGroups:NiagaraInt32=2
      Vector Group 0:Vector3f=<0.7 0.65 0.8>
      Vector Group 1:Vector3f=<0.7 0.35 0.8>
      Vector Group 2:Vector3f=<0 0 0>
      Vector Group 3:Vector3f=<0 0 0>
      Vector Group 4:Vector3f=<0 0 0>
      Vector Group 5:Vector3f=<0 0 0>
     )
     Float=3.5
    )
    Mesh Scale Mode=::NewEnumerator2
    UsePositionOffset=false
    Write Material Random=true
   )
   &PartitionParticles(
    A:NiagaraFloat=/Niagara/DynamicInputs/Integer/V2/Round_Float(Float:NiagaraFloat=&Sine(Normalized Angle=$Emitter.Age, Period=0.2, Scale=1, Bias=0.5))
    B:NiagaraFloat=&MakeFloatFromInt(Int=&SpawnGroup())
    Comparison Type:ENiagaraBooleanLogicOps_v2=::NewEnumerator4
    Partition Mode:ENiagara_PartitionMode=::NewEnumerator4
   )
   &SkeletalMeshLocation(
    Bone / Socket Index=&Add_Integer(A=$Particles.PartitionParticles.Partition, B=4)
    Bone / Socket Sampling Mode=::NewEnumerator5
    Interpolate Mesh Sampling=false
    MaskSpawnGroup=false
    Mesh Sampling Type=::NewEnumerator6
    Sampled Position Offset=<-47.5 0 -17.5>
    Skeletal Mesh=$System.SkeletalMesh
   )
   &AddVelocity(Velocity=&VC_Mask_Vectors(
     FrontEnd Value:Vector3f=<-1500 0 0>
     blsFrontEndPreview:NiagaraBool=false
     SpeedMin Value:Vector3f=<-800 0 0>
     SpeedMax Value:Vector3f=<-1 0 0>
     Speed Alpha:NiagaraFloat=$System.Speed_Alpha
    )) {enabled=false}
   &InitialMeshOrientation(
    Mesh Orientation Mode=::NewEnumerator2
    Rotation=&RandomRangeVector(Minimum=<0 0 0>, Maximum=<1 0 0>)
    Use Rotation Vector=true
   )
   /SetVariables(Particles.RenderVisTag:NiagaraInt32=&SpawnGroup()) {isSetParameters=true}
  ]
  particleUpdateScript [
   &FN_DisableTireFX(
    Mask Index Attribute:NiagaraInt32=$Particles.PartitionParticles.Partition
    Mask Index FL=2
    Mask Index FR=3
    Mask Index RL=0
    Mask Index RR=1
   )
   &ParticleState()
   &DynamicMaterialParameters(
    Index 0 Param 2=1
    Module.Index0Float0Write=true
    Module.Index0Float1Write=true
    Module.Index0Float2Write=true
    Module.Index0Float3Write=true
   ) {enabled=false}
   &ScaleColor(Scale RGB=&VectorFromFloat(Value=2), ScaleA=true, ScaleRGB=true)
   &SolveForcesAndVelocity(Clamp Velocity=false, Limit Acceleration=false)
  ]
  rendererClasses [&NiagaraMeshRendererProperties, &NiagaraSpriteRendererProperties]
  renderers [
   {rendererClass=&NiagaraMeshRendererProperties, rendererIndex=0}
   {rendererClass=&NiagaraMeshRendererProperties, rendererIndex=1}
   {rendererClass=&NiagaraSpriteRendererProperties, rendererIndex=2}
  ]
 }
 NE_Manga_Smoke {
  emitterSpawnScript [/SetVariables(Emitter.InGameEnabled:NiagaraBool=&Boolean_NotOperation(Bool=$User.bIsFrontendPreview)) {isSetParameters=true}]
  emitterUpdateScript [
   &EmitterState(Life Cycle Mode=::NewEnumerator1, UseLoopDelay=false)
   &SpawnRate(
    SpawnRate=&MakeCustomFloatFromBool(Bool=$User.bIsFrontendPreview, True Float=300, False Float=150)
    Use Spawn Probability=false
   )
   &SpawnRate(
    SpawnGroup=1
    SpawnRate=&MakeCustomFloatFromBool(Bool=$User.bIsFrontendPreview, True Float=120, False Float=90)
    Use Spawn Probability=false
   ) {module=SpawnRate001}
  ]
  particleSpawnScript [
   &VC_PaintedVariantsOverride(
    Black=<rgba 0.035 0.035 0.04 1>
    BurntSienna=<rgba 0.35 0.0733 0.0245 1>
    Cobalt=<rgba 0.08 0.17 0.8 1>
    Crimson=<rgba 0.5 0.0011 0.0011 1>
    DefaultColor=&VC_Dynamic_ColorBySpawnGroup(
     Color SG 0:LinearColor=<rgba 0.7 0.7 0.7 1>
     MaximumSpawnGroups:NiagaraInt32=2
     Color SG 1:LinearColor=<rgba 0.7 0.7 0.7 1>
     Color SG 2:LinearColor=<rgba 1 1 1 1>
     Color SG 3:LinearColor=<rgba 1 1 1 1>
     Color SG 4:LinearColor=<rgba 1 1 1 1>
    )
    ForestGreen=<rgba 0.06 0.4 0 1>
    Grey=<rgba 0.28 0.28 0.28 1>
    Lime=<rgba 0.55 1 0 1>
    Orange=<rgba 1 0.3488 0 1>
    Pink=<rgba 1 0.25 0.7273 1>
    Purple=<rgba 0.2708 0 0.65 1>
    Saffron=<rgba 0.9167 1 0 1>
    SkyBlue=<rgba 0 0.5333 0.8 1>
    TitaniumWhite=<rgba 0.95 0.95 1 1>
   )
   &InitializeParticle(
    Color Mode=::NewEnumerator0
    Lifetime=&MakeCustomFloatFromBool(
     Bool=$User.bIsFrontendPreview
     True Float=&VC_FloatBySpawnGroup(
      SpawnGroup or ExecIndex:NiagaraInt32=0
      CustomInt:NiagaraInt32=0
      MaximumSpawnGroups:NiagaraInt32=2
      Float SG 0:NiagaraFloat=&RandomRangeFloat(Minimum=0.25, Maximum=0.4)
      Float SG 1:NiagaraFloat=&RandomRangeFloat(Minimum=0.45, Maximum=1)
      Float SG 2:NiagaraFloat=0
      Float SG 3:NiagaraFloat=0
      Float SG 4:NiagaraFloat=0
      Float SG 5:NiagaraFloat=0
     )
     False Float=&RandomRangeFloat(Minimum=0.5, Maximum=0.55)
    )
    Sprite Rotation Mode=::NewEnumerator2
    Sprite Size Mode=::NewEnumerator3
    Uniform Sprite Size=&VC_FloatBySpawnGroup(
     SpawnGroup or ExecIndex:NiagaraInt32=0
     CustomInt:NiagaraInt32=0
     MaximumSpawnGroups:NiagaraInt32=2
     Float SG 0:NiagaraFloat=25
     Float SG 1:NiagaraFloat=&RandomRangeFloat(Minimum=20, Maximum=25)
     Float SG 2:NiagaraFloat=0
     Float SG 3:NiagaraFloat=0
     Float SG 4:NiagaraFloat=0
     Float SG 5:NiagaraFloat=0
    )
    UsePositionOffset=false
    Write Material Random=true
   )
   &PartitionParticles(Particles Per Partition=2, Partition Mode:ENiagara_PartitionMode=::NewEnumerator1)
   &SkeletalMeshLocation(
    Bone / Socket Index=&Add_Integer(A=$Particles.PartitionParticles.Partition, B=4)
    Bone / Socket Sampling Mode=::NewEnumerator5
    Interpolate Mesh Sampling=false
    MaskSpawnGroup=false
    Mesh Sampling Type=::NewEnumerator6
    Sampled Position Offset=<-48 0 -10>
    Skeletal Mesh=$System.SkeletalMesh
   )
   &ShapeLocation(
    MaskSpawnGroup=false
    Non Uniform Scale=&MakeCustomVectorFromBool(Bool=$User.bIsFrontendPreview, True Vector=<1 1 1>, False Vector=<1 1 1>)
    Override Seed=false
    Sphere Radius=&MakeCustomFloatFromBool(
     Bool=$User.bIsFrontendPreview
     True Float=&VC_FloatBySpawnGroup(
      SpawnGroup or ExecIndex:NiagaraInt32=0
      CustomInt:NiagaraInt32=0
      MaximumSpawnGroups:NiagaraInt32=2
      Float SG 0:NiagaraFloat=8
      Float SG 1:NiagaraFloat=20
      Float SG 2:NiagaraFloat=0
      Float SG 3:NiagaraFloat=0
      Float SG 4:NiagaraFloat=0
      Float SG 5:NiagaraFloat=0
     )
     False Float=&VC_FloatBySpawnGroup(
      SpawnGroup or ExecIndex:NiagaraInt32=0
      CustomInt:NiagaraInt32=0
      MaximumSpawnGroups:NiagaraInt32=2
      Float SG 0:NiagaraFloat=8
      Float SG 1:NiagaraFloat=20
      Float SG 2:NiagaraFloat=0
      Float SG 3:NiagaraFloat=0
      Float SG 4:NiagaraFloat=0
      Float SG 5:NiagaraFloat=0
     )
    )
    UseOwnerScale=true
   )
   &AddVelocity(Velocity=&MakeCustomVectorFromBool(
     Bool=$User.bIsFrontendPreview
     True Vector=&RandomRangeVector(Minimum=<-1500 -50 25>, Maximum=<-1500 50 50>)
     False Vector=&RandomRangeVector(Minimum=<-1050 -50 25>, Maximum=<-1050 50 50>)
    ))
  ]
  particleUpdateScript [
   &FN_DisableTireFX(
    Mask Index Attribute:NiagaraInt32=$Particles.PartitionParticles.Partition
    Mask Index FL=2
    Mask Index FR=3
    Mask Index RL=0
    Mask Index RR=1
   )
   &ParticleState()
   &ScaleSpriteSize(
    Scale Sprite Size Mode=::NewEnumerator0
    Uniform Scale Factor=&MakeCustomFloatFromBool(
     Bool=$User.bIsFrontendPreview
     True Float=&VC_FloatBySpawnGroup(
      SpawnGroup or ExecIndex:NiagaraInt32=0
      CustomInt:NiagaraInt32=0
      MaximumSpawnGroups:NiagaraInt32=2
      Float SG 0:NiagaraFloat=&FloatFromCurve(FloatCurve=DataInterface() {propertyValues {Curve {keys [
           {interpMode=RCIM_Cubic, tangentMode=RCTM_User, time=0, value=0.25}
           {arriveTangent=-0.0701, interpMode=RCIM_Cubic, leaveTangent=-0.0701, tangentMode=RCTM_User, time=0.15, value=0.5}
           {arriveTangent=0.1409, interpMode=RCIM_Cubic, leaveTangent=0.1409, tangentMode=RCTM_User, time=0.25, value=1.2}
           {arriveTangent=0.1385, interpMode=RCIM_Cubic, leaveTangent=0.1385, tangentMode=RCTM_User, time=0.8, value=3}
           {interpMode=RCIM_Cubic, tangentMode=RCTM_User, time=1, value=3}
          ]}}})
      Float SG 1:NiagaraFloat=&FloatFromCurve(FloatCurve=DataInterface() {propertyValues {Curve {keys [
           {interpMode=RCIM_Cubic, tangentMode=RCTM_User, time=0, value=0.25}
           {arriveTangent=-0.0701, interpMode=RCIM_Cubic, leaveTangent=-0.0701, tangentMode=RCTM_User, time=0.15, value=0.6}
           {arriveTangent=0.1409, interpMode=RCIM_Cubic, leaveTangent=0.1409, tangentMode=RCTM_User, time=0.25, value=1.2}
           {interpMode=RCIM_Cubic, tangentMode=RCTM_User, time=1, value=2.5}
          ]}}})
      Float SG 2:NiagaraFloat=0
      Float SG 3:NiagaraFloat=0
      Float SG 4:NiagaraFloat=0
      Float SG 5:NiagaraFloat=0
     )
     False Float=&VC_FloatBySpawnGroup(
      SpawnGroup or ExecIndex:NiagaraInt32=0
      CustomInt:NiagaraInt32=0
      MaximumSpawnGroups:NiagaraInt32=2
      Float SG 0:NiagaraFloat=&FloatFromCurve(
       FloatCurve=DataInterface() {propertyValues {Curve {
          defaultValue=1
          keys [
           {arriveTangent=2.2428, interpMode=RCIM_Cubic, leaveTangent=2.2428, tangentMode=RCTM_User, time=0, value=0}
           {arriveTangent=0.0224, interpMode=RCIM_Cubic, leaveTangent=0.0224, tangentMode=RCTM_User, time=0.6, value=1}
          ]
         }}}
       Scale Curve=4
      )
      Float SG 1:NiagaraFloat=&FloatFromCurve(
       FloatCurve=DataInterface() {propertyValues {Curve {
          defaultValue=1
          keys [
           {arriveTangent=2.6988, interpMode=RCIM_Cubic, leaveTangent=2.6988, tangentMode=RCTM_User, time=0.2, value=0}
           {arriveTangent=0.0224, interpMode=RCIM_Cubic, leaveTangent=0.0224, tangentMode=RCTM_User, time=0.6, value=1}
          ]
         }}}
       Scale Curve=4
      )
      Float SG 2:NiagaraFloat=0
      Float SG 3:NiagaraFloat=0
      Float SG 4:NiagaraFloat=0
      Float SG 5:NiagaraFloat=0
     )
    )
   )
   &SubUVAnimation(
    "Sprite Renderer "=DataInterface() {propertyValues {SpriteRenderer=&self:NE_Manga_Smoke_0.NiagaraSpriteRendererProperties_0}}
    UseEndFrame=false
    UseStartFrame=false
   )
   &DynamicMaterialParameters(
    Index 0 Param 2=&FloatFromCurve(
     FloatCurve=DataInterface() {propertyValues {Curve {keys [
         <at 0 0 RCIM_Cubic>
         {arriveTangent=-0.117, interpMode=RCIM_Cubic, leaveTangent=-0.117, tangentMode=RCTM_User, time=0.4, value=1.5}
         {arriveTangent=-0.0301, interpMode=RCIM_Cubic, leaveTangent=-0.0301, tangentMode=RCTM_User, time=0.7, value=1}
         <at 1 0 RCIM_Cubic>
        ]}}}
     Scale Curve=3
    )
    Index 0 Param 3=&VC_FloatBySpawnGroup(
     SpawnGroup or ExecIndex:NiagaraInt32=0
     CustomInt:NiagaraInt32=0
     MaximumSpawnGroups:NiagaraInt32=2
     Float SG 0:NiagaraFloat=&FloatFromCurve(FloatCurve=DataInterface() {propertyValues {Curve {keys [
          <at 0 1 RCIM_Cubic>
          {arriveTangent=0.0275, interpMode=RCIM_Cubic, leaveTangent=0.0275, tangentMode=RCTM_User, time=0.4, value=0}
          {arriveTangent=0.2242, interpMode=RCIM_Cubic, leaveTangent=0.2242, tangentMode=RCTM_User, time=0.6, value=0}
          <at 1 1 RCIM_Cubic>
         ]}}})
     Float SG 1:NiagaraFloat=&FloatFromCurve(FloatCurve=DataInterface() {propertyValues {Curve {keys [<at 0 1 RCIM_Cubic>, {arriveTangent=0.0275, interpMode=RCIM_Cubic, leaveTangent=0.0275, tangentMode=RCTM_User, time=0.4, value=0}]}}})
     Float SG 2:NiagaraFloat=0
     Float SG 3:NiagaraFloat=0
     Float SG 4:NiagaraFloat=0
     Float SG 5:NiagaraFloat=0
    )
    Index 0 Param 4=&VC_FloatBySpawnGroup(
     SpawnGroup or ExecIndex:NiagaraInt32=0
     CustomInt:NiagaraInt32=0
     MaximumSpawnGroups:NiagaraInt32=2
     Float SG 0:NiagaraFloat=1
     Float SG 1:NiagaraFloat=&RandomRangeFloat(Minimum=0, Maximum=1)
     Float SG 2:NiagaraFloat=0
     Float SG 3:NiagaraFloat=0
     Float SG 4:NiagaraFloat=0
     Float SG 5:NiagaraFloat=0
    )
    Module.Index0Float0Write=true
    Module.Index0Float1Write=true
    Module.Index0Float2Write=true
    Module.Index0Float3Write=true
   )
   &ScaleColor(
    Scale Alpha=&MakeCustomFloatFromBool(
     Bool=$User.bIsFrontendPreview
     True Float=&VC_FloatBySpawnGroup(
      SpawnGroup or ExecIndex:NiagaraInt32=0
      CustomInt:NiagaraInt32=0
      MaximumSpawnGroups:NiagaraInt32=2
      Float SG 0:NiagaraFloat=&FloatFromCurve(FloatCurve=DataInterface() {propertyValues {Curve {keys [
           {arriveTangent=-1.3333, interpMode=RCIM_Cubic, time=0.2, value=0.1}
           {arriveTangent=0.065, interpMode=RCIM_Cubic, leaveTangent=0.065, tangentMode=RCTM_User, time=0.3, value=1}
           {arriveTangent=-0.0092, interpMode=RCIM_Cubic, leaveTangent=-0.0092, tangentMode=RCTM_User, time=0.7, value=0.4}
           <at 1 0 RCIM_Cubic>
          ]}}})
      Float SG 1:NiagaraFloat=&FloatFromCurve(FloatCurve=DataInterface() {propertyValues {Curve {
          defaultValue=1
          keys [
           {arriveTangent=0.0221, interpMode=RCIM_Cubic, leaveTangent=0.0221, tangentMode=RCTM_User, time=0.125, value=1}
           {arriveTangent=0.2149, interpMode=RCIM_Cubic, leaveTangent=0.2149, tangentMode=RCTM_User, time=0.8, value=0.1}
           {arriveTangent=0.0609, interpMode=RCIM_Cubic, leaveTangent=0.0609, tangentMode=RCTM_User, time=1, value=0}
          ]
         }}})
      Float SG 2:NiagaraFloat=0
      Float SG 3:NiagaraFloat=0
      Float SG 4:NiagaraFloat=0
      Float SG 5:NiagaraFloat=0
     )
     False Float=&VC_FloatBySpawnGroup(
      SpawnGroup or ExecIndex:NiagaraInt32=0
      CustomInt:NiagaraInt32=0
      MaximumSpawnGroups:NiagaraInt32=2
      Float SG 0:NiagaraFloat=&FloatFromCurve(FloatCurve=DataInterface() {propertyValues {Curve {keys [
           {arriveTangent=-1.3333, interpMode=RCIM_Cubic, time=0.2, value=0.6}
           {arriveTangent=0.065, interpMode=RCIM_Cubic, leaveTangent=0.065, tangentMode=RCTM_User, time=0.3, value=1}
           {arriveTangent=-0.0092, interpMode=RCIM_Cubic, leaveTangent=-0.0092, tangentMode=RCTM_User, time=0.7, value=0.4}
           <at 1 0 RCIM_Cubic>
          ]}}})
      Float SG 1:NiagaraFloat=&FloatFromCurve(FloatCurve=DataInterface() {propertyValues {Curve {
          defaultValue=1
          keys [
           <at 0.1 0 RCIM_Cubic>
           {arriveTangent=0.0221, interpMode=RCIM_Cubic, leaveTangent=0.0221, tangentMode=RCTM_User, time=0.2, value=1}
           {arriveTangent=0.2149, interpMode=RCIM_Cubic, leaveTangent=0.2149, tangentMode=RCTM_User, time=0.8, value=0.1}
           {arriveTangent=0.0609, interpMode=RCIM_Cubic, leaveTangent=0.0609, tangentMode=RCTM_User, time=1, value=0}
          ]
         }}})
      Float SG 2:NiagaraFloat=0
      Float SG 3:NiagaraFloat=0
      Float SG 4:NiagaraFloat=0
      Float SG 5:NiagaraFloat=0
     )
    )
    Scale RGB=&VectorFromFloat(Value=1)
    ScaleA=true
    ScaleRGB=true
   )
   &InheritSourceMovement(
    Applied Position Scale=&Multiply_VectorByFloat(
     Vector=<1 1 1>
     Float=&FloatFromCurve(FloatCurve=DataInterface() {propertyValues {Curve {keys [
          {
           arriveTangentWeight=0.4252
           interpMode=RCIM_Cubic
           leaveTangentWeight=0.4252
           tangentMode=RCTM_User
           tangentWeightMode=RCTWM_WeightedBoth
           time=0.15
           value=1
          }
          {interpMode=RCIM_Cubic, tangentMode=RCTM_User, time=1, value=0}
         ]}}})
    )
    ApplyForce=false
    ApplyPosition=true
    ApplyPositionFalloff=false
   )
   &Drag(
    Drag=&MakeCustomFloatFromBool(
     Bool=$User.bIsFrontendPreview
     True Float=&VC_FloatBySpawnGroup(
      SpawnGroup or ExecIndex:NiagaraInt32=0
      CustomInt:NiagaraInt32=0
      MaximumSpawnGroups:NiagaraInt32=2
      Float SG 0:NiagaraFloat=1
      Float SG 1:NiagaraFloat=&FloatFromCurve(
       FloatCurve=DataInterface() {propertyValues {Curve {keys [
           <at 0 0 RCIM_Cubic>
           {arriveTangent=0.0407, interpMode=RCIM_Cubic, leaveTangent=0.0407, tangentMode=RCTM_User, time=0.45, value=0.2}
           <at 0.75 1 RCIM_Cubic>
          ]}}}
       Scale Curve=37.5
      )
      Float SG 2:NiagaraFloat=0
      Float SG 3:NiagaraFloat=0
      Float SG 4:NiagaraFloat=0
      Float SG 5:NiagaraFloat=0
     )
     False Float=0
    )
    Use Linear Drag=true
    Use Rotational Drag=true
   )
   &SolveForcesAndVelocity(Clamp Velocity=false, Limit Acceleration=false)
   &RotateAroundPoint(
    Radius:NiagaraFloat=&VC_FloatBySpawnGroup(
     SpawnGroup or ExecIndex:NiagaraInt32=0
     CustomInt:NiagaraInt32=0
     MaximumSpawnGroups:NiagaraInt32=2
     Float SG 0:NiagaraFloat=0
     Float SG 1:NiagaraFloat=1.5
     Float SG 2:NiagaraFloat=0
     Float SG 3:NiagaraFloat=0
     Float SG 4:NiagaraFloat=0
     Float SG 5:NiagaraFloat=0
    )
    Rotation Center:NiagaraPosition=$Particles.Position
    Rotation Rate:NiagaraFloat=&SelectFloatFromArray(
     Array Sampling Mode=::NewEnumerator0
     Direct Array Index=?
     Float Selection Array=DataInterface() {propertyValues {FloatData [-1,1,-1.5,1.5]}}
     Random Seed=0
    )
    X Axis:Vector3f=<0 1 0>
    Y Axis:Vector3f=<0 0 1>
   )
  ]
  rendererClasses [&NiagaraSpriteRendererProperties]
  renderers [{rendererClass=&NiagaraSpriteRendererProperties}]
 }
 NE_Manga_Smoke001 {
  emitterUpdateScript [
   &EmitterState(Life Cycle Mode=::NewEnumerator1, UseLoopDelay=false)
   &SpawnRate(
    SpawnRate=&MakeCustomFloatFromBool(Bool=$User.bIsFrontendPreview, True Float=300, False Float=150)
    Use Spawn Probability=false
   )
   &SpawnRate(
    SpawnGroup=1
    SpawnRate=&MakeCustomFloatFromBool(Bool=$User.bIsFrontendPreview, True Float=120, False Float=90)
    Use Spawn Probability=false
   ) {module=SpawnRate001}
  ]
  particleSpawnScript [
   &VC_PaintedVariantsOverride(
    Black=<rgba 0.035 0.035 0.04 1>
    BurntSienna=<rgba 0.35 0.0733 0.0245 1>
    Cobalt=<rgba 0.08 0.17 0.8 1>
    Crimson=<rgba 0.5 0.0011 0.0011 1>
    DefaultColor=&VC_Dynamic_ColorBySpawnGroup(
     Color SG 0:LinearColor=<rgba 0.7 0.7 0.7 1>
     MaximumSpawnGroups:NiagaraInt32=2
     Color SG 1:LinearColor=<rgba 0.7 0.7 0.7 1>
     Color SG 2:LinearColor=<rgba 1 1 1 1>
     Color SG 3:LinearColor=<rgba 1 1 1 1>
     Color SG 4:LinearColor=<rgba 1 1 1 1>
    )
    ForestGreen=<rgba 0.06 0.4 0 1>
    Grey=<rgba 0.28 0.28 0.28 1>
    Lime=<rgba 0.55 1 0 1>
    Orange=<rgba 1 0.3488 0 1>
    Pink=<rgba 1 0.25 0.7273 1>
    Purple=<rgba 0.2708 0 0.65 1>
    Saffron=<rgba 0.9167 1 0 1>
    SkyBlue=<rgba 0 0.5333 0.8 1>
    TitaniumWhite=<rgba 0.95 0.95 1 1>
   )
   &InitializeParticle(
    Color Mode=::NewEnumerator0
    Lifetime=&MakeCustomFloatFromBool(
     Bool=$User.bIsFrontendPreview
     True Float=&VC_FloatBySpawnGroup(
      SpawnGroup or ExecIndex:NiagaraInt32=0
      CustomInt:NiagaraInt32=0
      MaximumSpawnGroups:NiagaraInt32=2
      Float SG 0:NiagaraFloat=&RandomRangeFloat(Minimum=0.25, Maximum=0.4)
      Float SG 1:NiagaraFloat=&RandomRangeFloat(Minimum=0.45, Maximum=1)
      Float SG 2:NiagaraFloat=0
      Float SG 3:NiagaraFloat=0
      Float SG 4:NiagaraFloat=0
      Float SG 5:NiagaraFloat=0
     )
     False Float=&RandomRangeFloat(Minimum=0.5, Maximum=0.55)
    )
    Sprite Rotation Mode=::NewEnumerator2
    Sprite Size Mode=::NewEnumerator3
    Uniform Sprite Size=&VC_FloatBySpawnGroup(
     SpawnGroup or ExecIndex:NiagaraInt32=0
     CustomInt:NiagaraInt32=0
     MaximumSpawnGroups:NiagaraInt32=2
     Float SG 0:NiagaraFloat=25
     Float SG 1:NiagaraFloat=&RandomRangeFloat(Minimum=20, Maximum=25)
     Float SG 2:NiagaraFloat=0
     Float SG 3:NiagaraFloat=0
     Float SG 4:NiagaraFloat=0
     Float SG 5:NiagaraFloat=0
    )
    UsePositionOffset=false
    Write Material Random=true
   )
   &PartitionParticles(Particles Per Partition=2, Partition Mode:ENiagara_PartitionMode=::NewEnumerator1)
   &SkeletalMeshLocation(
    Bone / Socket Index=&Add_Integer(A=$Particles.PartitionParticles.Partition, B=4)
    Bone / Socket Sampling Mode=::NewEnumerator5
    Interpolate Mesh Sampling=false
    MaskSpawnGroup=false
    Mesh Sampling Type=::NewEnumerator6
    Sampled Position Offset=<-48 0 -20>
    Skeletal Mesh=$System.SkeletalMesh
   )
   &ShapeLocation(
    MaskSpawnGroup=false
    Non Uniform Scale=&MakeCustomVectorFromBool(Bool=$User.bIsFrontendPreview, True Vector=<1 1 1>, False Vector=<1 1 1>)
    Override Seed=false
    Sphere Radius=&MakeCustomFloatFromBool(
     Bool=$User.bIsFrontendPreview
     True Float=&VC_FloatBySpawnGroup(
      SpawnGroup or ExecIndex:NiagaraInt32=0
      CustomInt:NiagaraInt32=0
      MaximumSpawnGroups:NiagaraInt32=2
      Float SG 0:NiagaraFloat=8
      Float SG 1:NiagaraFloat=20
      Float SG 2:NiagaraFloat=0
      Float SG 3:NiagaraFloat=0
      Float SG 4:NiagaraFloat=0
      Float SG 5:NiagaraFloat=0
     )
     False Float=&VC_FloatBySpawnGroup(
      SpawnGroup or ExecIndex:NiagaraInt32=0
      CustomInt:NiagaraInt32=0
      MaximumSpawnGroups:NiagaraInt32=2
      Float SG 0:NiagaraFloat=8
      Float SG 1:NiagaraFloat=20
      Float SG 2:NiagaraFloat=0
      Float SG 3:NiagaraFloat=0
      Float SG 4:NiagaraFloat=0
      Float SG 5:NiagaraFloat=0
     )
    )
    UseOwnerScale=true
   )
   &AddVelocity(Velocity=&MakeCustomVectorFromBool(
     Bool=$User.bIsFrontendPreview
     True Vector=&RandomRangeVector(Minimum=<-1500 -50 25>, Maximum=<-1500 50 50>)
     False Vector=&RandomRangeVector(Minimum=<-1050 -50 25>, Maximum=<-1050 50 50>)
    ))
  ]
  particleUpdateScript [
   &FN_DisableTireFX(
    Mask Index Attribute:NiagaraInt32=$Particles.PartitionParticles.Partition
    Mask Index FL=2
    Mask Index FR=3
    Mask Index RL=0
    Mask Index RR=1
   )
   &ParticleState()
   &ScaleSpriteSize(
    Scale Sprite Size Mode=::NewEnumerator0
    Uniform Scale Factor=&MakeCustomFloatFromBool(
     Bool=$User.bIsFrontendPreview
     True Float=&VC_FloatBySpawnGroup(
      SpawnGroup or ExecIndex:NiagaraInt32=0
      CustomInt:NiagaraInt32=0
      MaximumSpawnGroups:NiagaraInt32=2
      Float SG 0:NiagaraFloat=&FloatFromCurve(FloatCurve=DataInterface() {propertyValues {Curve {keys [
           {interpMode=RCIM_Cubic, tangentMode=RCTM_User, time=0, value=0.25}
           {arriveTangent=-0.0701, interpMode=RCIM_Cubic, leaveTangent=-0.0701, tangentMode=RCTM_User, time=0.15, value=0.5}
           {arriveTangent=0.1409, interpMode=RCIM_Cubic, leaveTangent=0.1409, tangentMode=RCTM_User, time=0.25, value=1.2}
           {arriveTangent=0.1385, interpMode=RCIM_Cubic, leaveTangent=0.1385, tangentMode=RCTM_User, time=0.8, value=3}
           {interpMode=RCIM_Cubic, tangentMode=RCTM_User, time=1, value=3}
          ]}}})
      Float SG 1:NiagaraFloat=&FloatFromCurve(FloatCurve=DataInterface() {propertyValues {Curve {keys [
           {interpMode=RCIM_Cubic, tangentMode=RCTM_User, time=0, value=0.25}
           {arriveTangent=-0.0701, interpMode=RCIM_Cubic, leaveTangent=-0.0701, tangentMode=RCTM_User, time=0.15, value=0.6}
           {arriveTangent=0.1409, interpMode=RCIM_Cubic, leaveTangent=0.1409, tangentMode=RCTM_User, time=0.25, value=1.2}
           {interpMode=RCIM_Cubic, tangentMode=RCTM_User, time=1, value=2.5}
          ]}}})
      Float SG 2:NiagaraFloat=0
      Float SG 3:NiagaraFloat=0
      Float SG 4:NiagaraFloat=0
      Float SG 5:NiagaraFloat=0
     )
     False Float=&VC_FloatBySpawnGroup(
      SpawnGroup or ExecIndex:NiagaraInt32=0
      CustomInt:NiagaraInt32=0
      MaximumSpawnGroups:NiagaraInt32=2
      Float SG 0:NiagaraFloat=&FloatFromCurve(
       FloatCurve=DataInterface() {propertyValues {Curve {
          defaultValue=1
          keys [
           {arriveTangent=2.2428, interpMode=RCIM_Cubic, leaveTangent=2.2428, tangentMode=RCTM_User, time=0, value=0}
           {arriveTangent=0.0224, interpMode=RCIM_Cubic, leaveTangent=0.0224, tangentMode=RCTM_User, time=0.6, value=1}
          ]
         }}}
       Scale Curve=4
      )
      Float SG 1:NiagaraFloat=&FloatFromCurve(
       FloatCurve=DataInterface() {propertyValues {Curve {
          defaultValue=1
          keys [
           {arriveTangent=2.6988, interpMode=RCIM_Cubic, leaveTangent=2.6988, tangentMode=RCTM_User, time=0.2, value=0}
           {arriveTangent=0.0224, interpMode=RCIM_Cubic, leaveTangent=0.0224, tangentMode=RCTM_User, time=0.6, value=1}
          ]
         }}}
       Scale Curve=3
      )
      Float SG 2:NiagaraFloat=0
      Float SG 3:NiagaraFloat=0
      Float SG 4:NiagaraFloat=0
      Float SG 5:NiagaraFloat=0
     )
    )
   )
   &SubUVAnimation(
    "Sprite Renderer "=DataInterface() {propertyValues {SpriteRenderer=&self:NE_Manga_Smoke001_1.NiagaraSpriteRendererProperties_0}}
    UseEndFrame=false
    UseStartFrame=false
   )
   &DynamicMaterialParameters(
    Index 0 Param 2=&FloatFromCurve(FloatCurve=DataInterface() {propertyValues {Curve {keys [
         <at 0 0 RCIM_Cubic>
         {arriveTangent=-0.117, interpMode=RCIM_Cubic, leaveTangent=-0.117, tangentMode=RCTM_User, time=0.4, value=1.5}
         {arriveTangent=-0.0301, interpMode=RCIM_Cubic, leaveTangent=-0.0301, tangentMode=RCTM_User, time=0.7, value=1}
         <at 1 0 RCIM_Cubic>
        ]}}})
    Index 0 Param 3=&VC_FloatBySpawnGroup(
     SpawnGroup or ExecIndex:NiagaraInt32=0
     CustomInt:NiagaraInt32=0
     MaximumSpawnGroups:NiagaraInt32=1
     Float SG 0:NiagaraFloat=&FloatFromCurve(FloatCurve=DataInterface() {propertyValues {Curve {keys [
          <at 0 1 RCIM_Cubic>
          {arriveTangent=0.0275, interpMode=RCIM_Cubic, leaveTangent=0.0275, tangentMode=RCTM_User, time=0.4, value=0}
          {arriveTangent=0.2242, interpMode=RCIM_Cubic, leaveTangent=0.2242, tangentMode=RCTM_User, time=0.6, value=0}
          <at 1 1 RCIM_Cubic>
         ]}}})
     Float SG 1:NiagaraFloat=&FloatFromCurve(FloatCurve=DataInterface() {propertyValues {Curve {keys [<at 0 1 RCIM_Cubic>, {arriveTangent=0.0275, interpMode=RCIM_Cubic, leaveTangent=0.0275, tangentMode=RCTM_User, time=0.4, value=0}]}}})
     Float SG 2:NiagaraFloat=0
     Float SG 3:NiagaraFloat=0
     Float SG 4:NiagaraFloat=0
     Float SG 5:NiagaraFloat=0
    )
    Index 0 Param 4=&VC_FloatBySpawnGroup(
     SpawnGroup or ExecIndex:NiagaraInt32=0
     CustomInt:NiagaraInt32=0
     MaximumSpawnGroups:NiagaraInt32=2
     Float SG 0:NiagaraFloat=1
     Float SG 1:NiagaraFloat=&RandomRangeFloat(Minimum=0, Maximum=1)
     Float SG 2:NiagaraFloat=0
     Float SG 3:NiagaraFloat=0
     Float SG 4:NiagaraFloat=0
     Float SG 5:NiagaraFloat=0
    )
    Module.Index0Float0Write=true
    Module.Index0Float1Write=true
    Module.Index0Float2Write=true
    Module.Index0Float3Write=true
   )
   &ScaleColor(
    Scale Alpha=&MakeCustomFloatFromBool(
     Bool=$User.bIsFrontendPreview
     True Float=&VC_FloatBySpawnGroup(
      SpawnGroup or ExecIndex:NiagaraInt32=0
      CustomInt:NiagaraInt32=0
      MaximumSpawnGroups:NiagaraInt32=2
      Float SG 0:NiagaraFloat=&FloatFromCurve(FloatCurve=DataInterface() {propertyValues {Curve {keys [
           {arriveTangent=-1.3333, interpMode=RCIM_Cubic, time=0.2, value=0.1}
           {arriveTangent=0.065, interpMode=RCIM_Cubic, leaveTangent=0.065, tangentMode=RCTM_User, time=0.3, value=1}
           {arriveTangent=-0.0092, interpMode=RCIM_Cubic, leaveTangent=-0.0092, tangentMode=RCTM_User, time=0.7, value=0.4}
           <at 1 0 RCIM_Cubic>
          ]}}})
      Float SG 1:NiagaraFloat=&FloatFromCurve(FloatCurve=DataInterface() {propertyValues {Curve {
          defaultValue=1
          keys [
           {arriveTangent=0.0221, interpMode=RCIM_Cubic, leaveTangent=0.0221, tangentMode=RCTM_User, time=0.125, value=1}
           {arriveTangent=0.2149, interpMode=RCIM_Cubic, leaveTangent=0.2149, tangentMode=RCTM_User, time=0.8, value=0.1}
           {arriveTangent=0.0609, interpMode=RCIM_Cubic, leaveTangent=0.0609, tangentMode=RCTM_User, time=1, value=0}
          ]
         }}})
      Float SG 2:NiagaraFloat=0
      Float SG 3:NiagaraFloat=0
      Float SG 4:NiagaraFloat=0
      Float SG 5:NiagaraFloat=0
     )
     False Float=&VC_FloatBySpawnGroup(
      SpawnGroup or ExecIndex:NiagaraInt32=0
      CustomInt:NiagaraInt32=0
      MaximumSpawnGroups:NiagaraInt32=5
      Float SG 0:NiagaraFloat=&FloatFromCurve(FloatCurve=DataInterface() {propertyValues {Curve {keys [
           {arriveTangent=-1.3333, interpMode=RCIM_Cubic, time=0.2, value=0.6}
           {arriveTangent=0.065, interpMode=RCIM_Cubic, leaveTangent=0.065, tangentMode=RCTM_User, time=0.3, value=1}
           {arriveTangent=-0.0092, interpMode=RCIM_Cubic, leaveTangent=-0.0092, tangentMode=RCTM_User, time=0.7, value=0.4}
           <at 1 0 RCIM_Cubic>
          ]}}})
      Float SG 1:NiagaraFloat=&FloatFromCurve(FloatCurve=DataInterface() {propertyValues {Curve {
          defaultValue=1
          keys [
           <at 0.1 0 RCIM_Cubic>
           {arriveTangent=0.0221, interpMode=RCIM_Cubic, leaveTangent=0.0221, tangentMode=RCTM_User, time=0.2, value=1}
           {arriveTangent=0.2149, interpMode=RCIM_Cubic, leaveTangent=0.2149, tangentMode=RCTM_User, time=0.8, value=0.1}
           {arriveTangent=0.0609, interpMode=RCIM_Cubic, leaveTangent=0.0609, tangentMode=RCTM_User, time=1, value=0}
          ]
         }}})
      Float SG 2:NiagaraFloat=0
      Float SG 3:NiagaraFloat=0
      Float SG 4:NiagaraFloat=0
      Float SG 5:NiagaraFloat=0
     )
    )
    Scale RGB=&VectorFromFloat(Value=1)
    ScaleA=true
    ScaleRGB=true
   )
   &InheritSourceMovement(
    Applied Position Scale=&Multiply_VectorByFloat(
     Vector=<1 1 1>
     Float=&FloatFromCurve(FloatCurve=DataInterface() {propertyValues {Curve {keys [
          {
           arriveTangentWeight=0.4252
           interpMode=RCIM_Cubic
           leaveTangentWeight=0.4252
           tangentMode=RCTM_User
           tangentWeightMode=RCTWM_WeightedBoth
           time=0.15
           value=1
          }
          {interpMode=RCIM_Cubic, tangentMode=RCTM_User, time=1, value=0}
         ]}}})
    )
    ApplyForce=false
    ApplyPosition=true
    ApplyPositionFalloff=false
   )
   &Drag(
    Drag=&MakeCustomFloatFromBool(
     Bool=$User.bIsFrontendPreview
     True Float=&VC_FloatBySpawnGroup(
      SpawnGroup or ExecIndex:NiagaraInt32=0
      CustomInt:NiagaraInt32=0
      MaximumSpawnGroups:NiagaraInt32=2
      Float SG 0:NiagaraFloat=1
      Float SG 1:NiagaraFloat=&FloatFromCurve(
       FloatCurve=DataInterface() {propertyValues {Curve {keys [
           <at 0 0 RCIM_Cubic>
           {arriveTangent=0.0407, interpMode=RCIM_Cubic, leaveTangent=0.0407, tangentMode=RCTM_User, time=0.45, value=0.2}
           <at 0.75 1 RCIM_Cubic>
          ]}}}
       Scale Curve=37.5
      )
      Float SG 2:NiagaraFloat=0
      Float SG 3:NiagaraFloat=0
      Float SG 4:NiagaraFloat=0
      Float SG 5:NiagaraFloat=0
     )
     False Float=0
    )
    Use Linear Drag=true
    Use Rotational Drag=true
   )
   &SolveForcesAndVelocity(Clamp Velocity=false, Limit Acceleration=false)
   &RotateAroundPoint(
    Radius:NiagaraFloat=&VC_FloatBySpawnGroup(
     SpawnGroup or ExecIndex:NiagaraInt32=0
     CustomInt:NiagaraInt32=0
     MaximumSpawnGroups:NiagaraInt32=2
     Float SG 0:NiagaraFloat=0
     Float SG 1:NiagaraFloat=1.5
     Float SG 2:NiagaraFloat=0
     Float SG 3:NiagaraFloat=0
     Float SG 4:NiagaraFloat=0
     Float SG 5:NiagaraFloat=0
    )
    Rotation Center:NiagaraPosition=$Particles.Position
    Rotation Rate:NiagaraFloat=&SelectFloatFromArray(
     Array Sampling Mode=::NewEnumerator0
     Direct Array Index=?
     Float Selection Array=DataInterface() {propertyValues {FloatData [-1,1,-1.5,1.5]}}
     Random Seed=0
    )
    X Axis:Vector3f=<0 1 0>
    Y Axis:Vector3f=<0 0 1>
   )
  ]
  rendererClasses [&NiagaraSpriteRendererProperties]
  renderers [{rendererClass=&NiagaraSpriteRendererProperties}]
 }
 NE_VC_DriftTrail_Ribbon_Parent001 {
  emitterUpdateScript [
   &EmitterState()
   &VC_SpawnRibbon_v1(Number of Ribbons=2, SpawnRate=30)
  ]
  particleSpawnScript [
   &VC_PaintedVariantsOverride(
    Black=<rgba 0.38 0.38 0.38 1>
    BurntSienna=<rgba 0.65 0.4179 0.3185 1>
    Cobalt=<rgba 0.35 0.4258 1 1>
    Crimson=<rgba 0.8 0 0.0217 1>
    DefaultColor=<rgba 0.88 0.88 0.88 1>
    ForestGreen=<rgba 0 0.75 0.0379 1>
    Grey=<rgba 0.6 0.6 0.6 1>
    Lime=<rgba 0.8333 1 0 1>
    Orange=<rgba 1 0.675 0.35 1>
    Pink=<rgba 1 0.65 0.9125 1>
    Purple=<rgba 0.5417 0 0.65 1>
    Saffron=<rgba 0.9 0.87 0 1>
    SkyBlue=<rgba 0.55 0.8494 1 1>
    TitaniumWhite=<rgba 1 1 1 1>
   )
   &VC_InitializeRibbon_v1(
    Color=$Particles.Color
    Custom Offset=true
    Enable Skeletal Position=true
    Lifetime=1
    Local Vector Offset=<-70 0 -15>
    Offset Amount=1
    Ribbon Width:NiagaraFloat=&MakeCustomFloatFromBool(Bool=$User.bIsFrontendPreview, True Float=21, False Float=30)
   )
   &AddVelocity(Velocity=<-1000 0 0>) {module=AddVelocity001}
  ]
  particleUpdateScript [
   &FN_DisableTireFX(Mask Index Attribute:NiagaraInt32=$Particles.SpawnGroup, Mask Index FL=2, Mask Index FR=3, Mask Index RL=0, Mask Index RR=1)
   &ParticleState()
   &InheritSourceMovement(
    Applied Position Scale=&Multiply_VectorByFloat(
     Vector=<1 1 1>
     Float=&FloatFromCurve(FloatCurve=DataInterface() {propertyValues {Curve {keys [
          {
           arriveTangentWeight=0.4252
           interpMode=RCIM_Cubic
           leaveTangentWeight=0.4252
           tangentMode=RCTM_User
           tangentWeightMode=RCTWM_WeightedBoth
           time=0.15
           value=1
          }
          {interpMode=RCIM_Cubic, tangentMode=RCTM_User, time=1, value=0}
         ]}}})
    )
    ApplyForce=false
    ApplyPosition=true
    ApplyPositionFalloff=false
   )
   &ScaleColor(
    Scale Alpha=&FloatFromCurve(FloatCurve=DataInterface() {propertyValues {Curve {keys [
         <at 0 -0.1 RCIM_Cubic>
         {arriveTangent=0.0132, interpMode=RCIM_Cubic, leaveTangent=0.0132, tangentMode=RCTM_User, time=0.15, value=-0.5}
         {arriveTangent=-0.0413, interpMode=RCIM_Cubic, leaveTangent=-0.0413, tangentMode=RCTM_User, time=0.2, value=-0.5}
         <at 1 -0.125 RCIM_Cubic>
        ]}}})
    Scale RGB=&MakeVectorFromLinearColorRGB(Color=$Particles.Color)
    ScaleA=true
    ScaleRGB=true
   )
   &SolveForcesAndVelocity(Clamp Velocity=false, Limit Acceleration=false)
   &VC_UpdateRibbon_v1(Enable Skeletal Position=true)
  ]
  rendererClasses [&NiagaraRibbonRendererProperties]
  renderers [{rendererClass=&NiagaraRibbonRendererProperties}]
 }
 NE_VC_DriftTrail_WheelWrap_Parent001 {
  emitterUpdateScript [
   &EmitterState(
    Inactive Response=::NewEnumerator1
    Life Cycle Mode=::NewEnumerator1
    Loop Behavior=::NewEnumerator1
    Loop Duration Mode=::NewEnumerator1
    UseLoopDelay=false
   )
   &SpawnBurst_Instantaneous(Spawn Count=2, Use Spawn Probability=false, UseLoopCountLimit=false)
  ]
  particleSpawnScript [
   &VC_PaintedVariantsOverride(
    Black=<rgba 0.035 0.035 0.035 1>
    BurntSienna=<rgba 0.5 0.0976 0.05 1>
    Cobalt=<rgba 0.1 0.205 1 1>
    Crimson=<rgba 0.5 0.0011 0.0011 1>
    ForestGreen=<rgba 0.05 0.5 0 1>
    Grey=<rgba 0.314 0.314 0.33 1>
    Lime=<rgba 0.8333 1 0 1>
    Orange=<rgba 1 0.5 0 1>
    Pink=<rgba 1 0.25 0.8125 1>
    Purple=<rgba 0.2042 0 0.5 1>
    Saffron=<rgba 0.825 0.9 0 1>
    SkyBlue=<rgba 0 0.5323 0.8 1>
    TitaniumWhite=<rgba 1 1 1 1>
   )
   &InitializeParticle(
    Color Mode=::NewEnumerator0
    Mesh Scale Mode=::NewEnumerator4
    Sprite Rotation Mode=::NewEnumerator2
    Sprite Size Mode=::NewEnumerator3
    UsePositionOffset=false
    Write Material Random=true
   )
   &PartitionParticles()
   &SkeletalMeshLocation(
    Bone / Socket Index=&Add_Integer(A=$Particles.PartitionParticles.Partition, B=4)
    Bone / Socket Sampling Mode=::NewEnumerator5
    Interpolate Mesh Sampling=false
    MaskSpawnGroup=false
    Mesh Sampling Type=::NewEnumerator6
    Skeletal Mesh=$System.SkeletalMesh
   )
  ]
  particleUpdateScript [
   &FN_DisableTireFX(
    Mask Index Attribute:NiagaraInt32=$Particles.PartitionParticles.Partition
    Mask Index FL=2
    Mask Index FR=3
    Mask Index RL=0
    Mask Index RR=1
   )
   &VC_WheelWrapScale(Wheel Index Attribute:NiagaraInt32=&Add_Integer(A=$Particles.PartitionParticles.Partition, B=2))
   &ParticleState(Kill Particles When Lifetime Has Elapsed=false)
   &ScaleColor(
    Scale Alpha=&FloatFromCurve(FloatCurve=DataInterface() {propertyValues {Curve {keys [
         <at 0 -0.1 RCIM_Cubic>
         {arriveTangent=0.0132, interpMode=RCIM_Cubic, leaveTangent=0.0132, tangentMode=RCTM_User, time=0.15, value=-0.5}
         {arriveTangent=-0.0413, interpMode=RCIM_Cubic, leaveTangent=-0.0413, tangentMode=RCTM_User, time=0.2, value=-0.5}
         <at 1 -0.125 RCIM_Cubic>
        ]}}})
    ScaleA=true
    ScaleRGB=true
   ) {enabled=false}
   &UpdateMeshOrientation(Rotation Coordinate Space=::NewEnumerator2, Rotation Vector=<0 1 0>) {enabled=false}
   &SkeletalMeshLocation(
    Bone / Socket Index=&Add_Integer(A=$Particles.PartitionParticles.Partition, B=4)
    Bone / Socket Sampling Mode=::NewEnumerator5
    MaskSpawnGroup=false
    Mesh Sampling Type=::NewEnumerator6
    Skeletal Mesh=$System.SkeletalMesh
   ) {module=SkeletalMeshLocation001}
   &SolveForcesAndVelocity(Clamp Velocity=false, Limit Acceleration=false)
  ]
  rendererClasses [&NiagaraMeshRendererProperties]
  renderers [{rendererClass=&NiagaraMeshRendererProperties, rendererIndex=0}]
 }
}
summary {
 systemName=NS_DriftTrail_Mangaoutline
 emitters [
  {emitterName=NE_Manga_Smoke, rendererClasses [&NiagaraSpriteRendererProperties]}
  {emitterName=NE_Manga_Smoke001, rendererClasses [&NiagaraSpriteRendererProperties]}
  {emitterName=NE_VC_DriftTrail_Ribbon_Parent001, rendererClasses [&NiagaraRibbonRendererProperties]}
  {
   emitterName=NE_Manga_Burst001
   rendererClasses [&NiagaraMeshRendererProperties, &NiagaraSpriteRendererProperties]
  }
  {emitterName=NE_VC_DriftTrail_WheelWrap_Parent001, rendererClasses [&NiagaraMeshRendererProperties]}
 ]
 userVariables [
  User.bIsFrontendPreview {type=NiagaraBool}
  User.bIsPainted {type=NiagaraBool}
  User.ColorVariantID {type=NiagaraInt32}
  User.offset {type=Vector3f, description="offset from wheels(set through cosmetic template)", defaultValue {struct=Vector3f, value=<0 0 0>}}
  User.PaintedColor {type=LinearColor, description="", defaultValue {struct=LinearColor, value=<rgba 1 1 1 1>}}
  User.PaintedScalar {type=NiagaraFloat, defaultValue=1, description=""}
  User.VehicleFlags {type=NiagaraInt32, defaultValue=806289408}
  User.Wheel_Radius_Array {
   type=NiagaraDataInterfaceArrayFloat
   defaultValue {
    struct=NiagaraExt_VariableValue_DataInterface
    value {
     dataInterface=&self:NiagaraDataInterfaceArrayFloat_1
     dataInterfaceClass=&NiagaraDataInterfaceArrayFloat
    }
   }
  }
  User.Wheel_Width_Array {
   type=NiagaraDataInterfaceArrayFloat
   defaultValue {
    struct=NiagaraExt_VariableValue_DataInterface
    value {
     dataInterface=&self:NiagaraDataInterfaceArrayFloat_0
     dataInterfaceClass=&NiagaraDataInterfaceArrayFloat
    }
   }
  }
  User.WheelTouchingArray {
   type=NiagaraDataInterfaceArrayFloat
   defaultValue {
    struct=NiagaraExt_VariableValue_DataInterface
    value {
     dataInterface=&self:NiagaraDataInterfaceArrayFloat_2
     dataInterfaceClass=&NiagaraDataInterfaceArrayFloat
    }
   }
  }
 ]
}
