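# Unlit masked Niagara sprite material that draws hand-sketched ink smoke for the manga-outline drift trail.
# Sketch texture green becomes a line pattern shaped by EdgeThickness, flipped between inverted and normal by particle color brightness and Transition, then tinted by particle color.
# Texture alpha, cloud noise and a tiled radial gradient build a dissolve mask driven by particle alpha.
# TextureSwap blends identical inputs; RadialDistort and the flow normal sample are unused.
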Material /VehicleCosmetics/DriftTrail/DriftTrail_Mangaoutline/Materials/M_InkDrawing_Smoke
graph {
 properties {
  bAutomaticallySetUsageInEditor=false
  bUsedWithNiagaraSprites=true
  blendMode=BLEND_Masked
  refractionMethod=RM_IndexOfRefraction
  shadingModel=MSM_Unlit
  usedShadingModels=Unlit
 }
 nodes [
  n1 = TextureSample!(UVs=TextureCoordinate(), Tex, Level, Apply View MipBias) {mipValueMode=TMVM_MipLevel, texture=/DealershipSharedLibrary/FX/Textures/TexturePackage/T_VCVFX_SketchSmoke_RGB}
  n2 = Ceil(None=DynamicParameter() {paramNames [RadialDistort, EdgeThickness, TextureSwap, Transition]}.TextureSwap)
  n3 = LinearInterpolate(A=@n1.G, B=@n1.G, Alpha=@n2)
  MaterialExpressionConstantClamp_4 = Saturate(None=ConstantBiasScale(None=Multiply(A=@n3, B=@n3)) {bias=-0.6, scale=3})
  MaterialExpressionConstantClamp_5 = Saturate(None=Multiply(A=@n3, B=Constant() {r=600}))
  n4 = LinearInterpolate(
   A=Multiply(A=@MaterialExpressionConstantClamp_4, B=@MaterialExpressionConstantClamp_4)
   B=Multiply(A=@MaterialExpressionConstantClamp_5, B=@MaterialExpressionConstantClamp_5)
   Alpha=DynamicParameter() {paramNames [RadialDistort, EdgeThickness, TextureSwap, Transition]}.EdgeThickness
  )
  MaterialExpressionConstantClamp_6 = Saturate(None=OneMinus(None=@n4))
  n5 = LinearInterpolate(A=@n1.A, B=@n1.A, Alpha=@n2)
  n6 = Multiply(A=@MaterialExpressionConstantClamp_6, B=@n5)
  n7 = Multiply(A=@n4, B=@n5)
  MaterialExpressionVertexColor_1 = ParticleColor()
  n8 = Add(A=@MaterialExpressionVertexColor_1.R, B=@MaterialExpressionVertexColor_1.R)
  MaterialExpressionConstantClamp_7 = Saturate(None=Divide(A=Add(A=@n8, B=@MaterialExpressionVertexColor_1.B), B=Constant() {r=3}))
  n9 = Constant()
  n10 = DynamicParameter() {paramNames [RadialDistort, EdgeThickness, TextureSwap, Transition]}
  MaterialExpressionConstantClamp_8 = Saturate(None=Multiply(A=Constant() {r=200000000753664}, B=@n10.Transition))
  n11 = If(
   A=@MaterialExpressionConstantClamp_7
   B=Constant() {r=0.78}
   "A > B"=@n9
   "A == B"=@n9
   "A < B"=If(
    A=@n10.Transition
    B=Constant() {r=0.49}
    "A > B"=@MaterialExpressionConstantClamp_8
    "A == B"=@MaterialExpressionConstantClamp_8
    "A < B"=Constant()
   )
  )
  MaterialFunctionCall(Emissive=Multiply(A=LinearInterpolate(A=@n6, B=@n7, Alpha=@n11), B=ParticleColor().RGB)) {materialFunction=&MF_FX_EmissiveAdjustments}
  n12 = TextureSample!(UVs=TextureCoordinate() {uTiling=2, vTiling=2}, Tex, Apply View MipBias) {texture=/VehicleCosmetics/FX/Textures/Noise/T_VCVFX_Clouds_A}
  n13 = Add(A=Frac(None=TextureCoordinate() {uTiling=2, vTiling=2}), B=Constant() {r=-0.5})
  n14 = DotProduct(A=@n13, B=@n13)
  MaterialExpressionConstantClamp_3 = Saturate(None=ConstantBiasScale(None=OneMinus(None=@n14)) {bias=-0.8, scale=1})
  MaterialExpressionConstantClamp_2 = Saturate(None=Multiply(A=@MaterialExpressionConstantClamp_3, B=@MaterialExpressionConstantClamp_3))
  MaterialExpressionConstantClamp_1 = Saturate(None=Add(A=ConstantBiasScale(None=@n12.RGB) {bias=0.5}, B=@MaterialExpressionConstantClamp_2))
  MaterialExpressionVertexColor_0 = ParticleColor()
  n15 = OneMinus(None=Add(A=@MaterialExpressionVertexColor_0.A, B=Constant() {r=0.2}))
  MaterialExpressionConstantClamp_0 = Saturate(None=Sine(None=Add(A=@MaterialExpressionConstantClamp_1, B=@n15)) {period=2})
  Multiply(A=Multiply(A=@n5, B=Constant() {r=5}), B=@MaterialExpressionConstantClamp_0)
  TextureSample(UVs=TextureCoordinate() {uTiling=2, vTiling=2}) {texture=/DealershipSharedLibrary/FX/Textures/Noise/T_VCVFX_Flow02_N}
 ]
}
