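# Masked unlit material for a manga-outline drift trail effect on Niagara ribbons and mesh particles.
# Panning noise and cloud textures, shaped by UV edge gradients and hard bias-scale clamps, build the trail mask and layered streaks, with MaskFade and ColorFade scalars.
# An If on averaged particle red, red again and blue (green is skipped) selects untinted or particle-tinted color for MF_FX_EmissiveAdjustments.
# Particle alpha is added to the mask.
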Material /VehicleCosmetics/DriftTrail/DriftTrail_Mangaoutline/Materials/M_MangaTrail
graph {
 properties {
  bAutomaticallySetUsageInEditor=false
  bUsedWithNiagaraMeshParticles=true
  bUsedWithNiagaraRibbons=true
  blendMode=BLEND_Masked
  refractionMethod=RM_PixelNormalOffset
  shadingModel=MSM_Unlit
  usedShadingModels=Unlit
 }
 nodes [
  MaterialExpressionVertexColor_2 = ParticleColor()
  n1 = TextureCoordinate()
  n2 = ComponentMask(None=@n1) {r=true}
  n3 = OneMinus(None=@n2)
  n4 = ScalarParameter() {defaultValue=5, parameterName=MaskFade}
  n5 = ComponentMask(None=@n1) {g=true}
  n6 = Multiply(A=@n5, B=OneMinus(None=@n5))
  n7 = Constant() {r=3}
  n8 = TextureSample(UVs=Panner(Coordinate=TextureCoordinate() {coordinateIndex=1, vTiling=0.15}) {speedX=-1, speedY=0.15}) {texture=/DealershipSharedLibrary/FX/Textures/Noise/T_VCVFX_CloudsSoft_RGB}
  n9 = TextureSample(UVs=Panner(Coordinate=TextureCoordinate() {coordinateIndex=1, vTiling=0.15}) {speedX=-1, speedY=-0.1}) {texture=&T_VCVFX_CloudsSoft_RGB}
  n10 = AppendVector(A=@n8.B, B=@n9.G)
  n11 = Desaturation(None=Multiply(A=Multiply(A=@n6, B=@n7), B=@n10))
  n12 = Constant() {r=3}
  n13 = TextureCoordinate()
  n14 = Constant() {r=1}
  n15 = ComponentMask(None=Panner(Coordinate=Rotator(Coordinate=@n13, Time=@n14) {speed=0.5}, Time=@n14) {speedY=-0.15}) {g=true}
  n16 = OneMinus(None=@n15)
  MaterialExpressionConstantClamp_13 = Saturate(None=Multiply(A=@n12, B=Multiply(A=@n15, B=@n16)))
  n17 = TextureCoordinate()
  n18 = Constant() {r=1}
  n19 = ComponentMask(None=Panner(Coordinate=Rotator(Coordinate=@n17, Time=@n18) {speed=-0.5}, Time=@n18) {speedY=0.15}) {g=true}
  n20 = OneMinus(None=@n19)
  MaterialExpressionConstantClamp_14 = Saturate(None=Multiply(A=Multiply(A=@n19, B=@n20), B=Constant() {r=3}))
  n21 = Multiply(A=@MaterialExpressionConstantClamp_13, B=@MaterialExpressionConstantClamp_14)
  n22 = Constant() {r=20}
  n23 = ComponentMask(None=@n13) {r=true}
  n24 = OneMinus(None=@n23)
  n25 = Multiply(A=Multiply(A=Multiply(A=@n3, B=@n4), B=@n11), B=Multiply(A=@n24, B=Multiply(A=@n21, B=@n22)))
  MaterialExpressionConstantClamp_22 = Saturate(None=ConstantBiasScale(None=@n25) {bias=-0.05, scale=5})
  Add(A=@MaterialExpressionVertexColor_2.A, B=@MaterialExpressionConstantClamp_22)
  n26 = TextureCoordinate() {coordinateIndex=1, uTiling=0.1}
  n27 = Multiply(A=Constant() {r=0.5}, B=Time())
  n28 = TextureCoordinate() {coordinateIndex=1, uTiling=0.05, vTiling=0.5}
  n29 = Add(
   A=TextureSample(UVs=Panner(Coordinate=@n26, Time=@n27) {speedX=-1, speedY=-0.15}) {texture=/DealershipSharedLibrary/FX/Textures/Noise/T_VCVFX_NoiseSpots_RGB}.R
   B=TextureSample(UVs=Panner(Coordinate=@n28, Time=@n27) {speedX=-1}) {texture=&T_VCVFX_NoiseSpots_RGB}.G
  )
  n30 = Constant() {r=10}
  n31 = OneMinus(None=Multiply(A=@n2, B=@n11))
  n32 = Multiply(A=Multiply(A=ScalarParameter() {defaultValue=5, parameterName=ColorFade}, B=@n3) {constA=5}, B=@n11)
  MaterialExpressionConstantClamp_17 = Saturate(None=ConstantBiasScale(None=@n32) {bias=-0.2, scale=6})
  MaterialExpressionConstantClamp_16 = Saturate(None=ConstantBiasScale(None=Multiply(A=@n31, B=@MaterialExpressionConstantClamp_17)) {bias=-0.3, scale=3})
  MaterialExpressionConstantClamp_15 = Saturate(None=OneMinus(None=@MaterialExpressionConstantClamp_16))
  n33 = Multiply(A=Multiply(A=@n24, B=@n23), B=@n12)
  n34 = Multiply(A=@n24, B=Multiply(A=Multiply(A=@n33, B=@n21), B=Constant() {r=2}))
  n35 = Multiply(A=Multiply(A=Multiply(A=@n29, B=@n30), B=@MaterialExpressionConstantClamp_15), B=@n34)
  MaterialExpressionConstantClamp_18 = Saturate(None=ConstantBiasScale(None=@n32) {bias=-0.2, scale=1})
  n36 = TextureSample(UVs=Panner(Coordinate=TextureCoordinate() {coordinateIndex=1, uTiling=0.5}) {speedX=-1.25, speedY=0.15}) {texture=&T_VCVFX_NoiseSpots_RGB}
  n37 = TextureSample(UVs=Panner(Coordinate=TextureCoordinate() {coordinateIndex=1, uTiling=2}) {speedX=-1.5, speedY=-0.1}) {texture=&T_VCVFX_CloudsSoft_RGB}
  MaterialExpressionConstantClamp_20 = Saturate(None=ConstantBiasScale(None=Add(A=@n36.R, B=@n37.G)) {bias=-0.5, scale=60})
  MaterialExpressionConstantClamp_19 = Saturate(None=OneMinus(None=@MaterialExpressionConstantClamp_20))
  n38 = Multiply(
   A=Multiply(A=Multiply(A=@MaterialExpressionConstantClamp_17, B=@MaterialExpressionConstantClamp_18), B=Constant() {r=3})
   B=Multiply(A=@MaterialExpressionConstantClamp_16, B=@MaterialExpressionConstantClamp_19)
  )
  n39 = Multiply(A=Constant() {r=3}, B=Add(A=@n35, B=Multiply(A=@n38, B=@n34)))
  MaterialExpressionVertexColor_3 = ParticleColor()
  n40 = Add(A=@MaterialExpressionVertexColor_3.R, B=@MaterialExpressionVertexColor_3.R)
  MaterialExpressionConstantClamp_21 = Saturate(None=Divide(A=Add(A=@n40, B=@MaterialExpressionVertexColor_3.B), B=Constant() {r=3}))
  n41 = Constant() {r=1}
  n42 = Constant()
  n43 = Constant() {r=1}
  MaterialFunctionCall(Emissive=LinearInterpolate(
    A=Add(A=@n35, B=@n25)
    B=Multiply(A=@n39, B=@MaterialExpressionVertexColor_2.RGB)
    Alpha=If(A=@MaterialExpressionConstantClamp_21, B=@n41, "A > B"=@n42, "A == B"=@n42, "A < B"=@n43)
   )) {materialFunction=&MF_FX_EmissiveAdjustments}
  n44 = Constant() {r=0.5}
  Reroute(None=AppendVector(A=Multiply(A=@n10, B=@n44), B=Add(A=@n8.B, B=@n9.B)))
 ]
}
