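# Masked, unlit, two-sided Niagara particle material for toon smoke bursts in the manga-outline drift trail.
# Blends two samples of a sketch smoke texture by the dynamic Alpha parameter, the second mirrored and statically offset through a constant-time Panner.
# Boosted, inverted color tints particle color into MF_FX_EmissiveAdjustments, while boosted texture alpha multiplies particle alpha.
# BlackCore has no effect because both its lerps have identical inputs.
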Material /VehicleCosmetics/DriftTrail/DriftTrail_Mangaoutline/Materials/M_ToonBurst
graph {
 properties {
  bAutomaticallySetUsageInEditor=false
  bUsedWithNiagaraMeshParticles=true
  bUsedWithNiagaraSprites=true
  blendMode=BLEND_Masked
  shadingModel=MSM_Unlit
  twoSided=true
  usedShadingModels=Unlit
 }
 nodes [
  MaterialExpressionMeshEmitterVertexColor_0 = ParticleColor()
  n1 = TextureSample() {texture=/DealershipSharedLibrary/FX/Textures/TexturePackage/T_VCVFX_SketchSmokeBurst_RGB}
  MaterialExpressionMeshEmitterDynamicParameter_0 = DynamicParameter() {paramNames [Alpha,BlackCore,Param3,Param4]}
  n2 = LinearInterpolate(A=@n1.R, B=@n1.R, Alpha=@MaterialExpressionMeshEmitterDynamicParameter_0.BlackCore)
  n3 = TextureSample(UVs=Panner(Coordinate=TextureCoordinate() {uTiling=-0.65, vTiling=0.9}, Time=Constant() {r=1}) {speedX=0.82, speedY=0.08}) {texture=&T_VCVFX_SketchSmokeBurst_RGB}
  n4 = LinearInterpolate(A=@n3.RGB, B=@n3.RGB, Alpha=@MaterialExpressionMeshEmitterDynamicParameter_0.BlackCore)
  MaterialExpressionConstantClamp_11 = Saturate(None=Multiply(A=Constant() {r=10}, B=LinearInterpolate(A=@n2, B=@n4, Alpha=@MaterialExpressionMeshEmitterDynamicParameter_0.Alpha)))
  MaterialExpressionConstantClamp_10 = Saturate(None=OneMinus(None=@MaterialExpressionConstantClamp_11))
  MaterialFunctionCall(Emissive=Multiply(A=@MaterialExpressionMeshEmitterVertexColor_0.RGB, B=@MaterialExpressionConstantClamp_10)) {materialFunction=&MF_FX_EmissiveAdjustments}
  MaterialExpressionConstantClamp_12 = Saturate(None=Multiply(A=LinearInterpolate(A=@n1.A, B=@n3.A, Alpha=@MaterialExpressionMeshEmitterDynamicParameter_0.Alpha), B=Constant() {r=5}))
  Multiply(A=@MaterialExpressionMeshEmitterVertexColor_0.A, B=@MaterialExpressionConstantClamp_12)
  TextureCoordinate() {uTiling=-0.65, vTiling=0.9}
 ]
}
